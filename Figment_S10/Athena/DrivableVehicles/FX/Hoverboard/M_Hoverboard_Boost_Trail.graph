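# Unlit, two-sided translucent material for the hoverboard boost trail, built procedurally from UV sphere masks without textures.
# The Width dynamic parameter widens and hardens the band across V while dimming its center, leaving brighter edges.
# Particle alpha scales the mask, which also tints particle color toward DarkColor along U.
# Depth fade softens intersections except on ES2/ES3_1; the declared InnerTranslucency parameter is unused.
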Material /Figment_S10/Athena/DrivableVehicles/FX/Hoverboard/M_Hoverboard_Boost_Trail
graph {
 properties {
  bUsedWithBeamTrails=true
  bUsedWithMeshDeformer=true
  bUsedWithParticleSprites=true
  bUsedWithSkeletalMesh=true
  blendMode=BLEND_Translucent
  pixelDepthOffsetMode=PDOM_Legacy
  shadingModel=MSM_Unlit
  twoSided=true
  usedShadingModels=Unlit
 }
 nodes [
  MaterialFunctionCall() {materialFunction=&DebugTimeSine}
  n1 = ParticleColor()
  n2 = TextureCoordinate()
  n3 = ComponentMask(None=@n2) {r=true}
  n4 = Constant() {r=0.5}
  n5 = ComponentMask(None=@n2) {g=true}
  n6 = DynamicParameter() {paramNames [Width, InnerTranslucency, Param3, Param4]}
  n7 = SphereMask(A=@n5, B=@n4, Radius=LinearInterpolate(Alpha=@n6.Width) {constA=0.01, constB=0.5}, Hardness=LinearInterpolate(Alpha=@n6.Width) {constB=0.9}) {hardnessPercent=70}
  n8 = SphereMask(A=@n5, B=@n4, Radius=@n4) {attenuationRadius=0.4, hardnessPercent=0}
  n9 = LinearInterpolate(
   A=LinearInterpolate(Alpha=Multiply(A=@n6.Width, B=@n6.Width)) {constA=1, constB=0.01}
   Alpha=Power(Base=OneMinus(None=@n8)) {constExponent=2.5}
  ) {constA=0.04}
  n10 = Add(A=@n9) {constB=0.1}
  n11 = Multiply(A=@n1.A, B=Multiply(A=SphereMask(A=@n3, B=@n4, Radius=@n4) {hardnessPercent=90}, B=Multiply(A=@n7, B=@n10)))
  n12 = DepthFade() {fadeDistanceDefault=30}
  n13 = Constant() {r=1}
  Multiply(A=@n11, B=FeatureLevelSwitch(Default=@n12, ES2=@n13, ES3_1=@n13, SM4_REMOVED=@n12))
  n14 = VectorParameter() {parameterName=DarkColor, defaultValue=<rgba 0.1718 0.2044 0.5 0>}
  n15 = Multiply(A=ComponentMask(None=TextureCoordinate()) {r=true}, B=@n11)
  Multiply(A=Multiply(A=LinearInterpolate(B=@n14.RGB, Alpha=@n15) {constA=1}, B=@n1.RGB), B=@n9)
 ]
}
